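# Animation blueprint for the face accessory part of the PixieParts ScrewArms character, adding springy physics to its robot shoulder pads.
# It copies the attached parent's pose, layers clavicle, arm, hand and shoulder-pad bones from the body mesh, and caches the result.
# Two Anim Dynamics chains simulate both pads relative to the upper arms, with bIsInFrontEnd selecting between their tunings.
# On customization load, it stores the body mesh and makes it a tick prerequisite.
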AnimBlueprint /BRCosmetics/Characters/Player/Male/Medium/Bodies/M_MED_PixieParts_ScrewArms/Meshes/Parts/M_MED_PixieParts_ScrewArms_FaceAcc_AnimBP
target_skeleton=&/M_MED_PixieParts_ScrewArms_FaceAcc_Skeleton
graphs [
 AnimGraph {nodes [
   n1 = Anim Dynamics - Bone: robot_shoulderpad_r!(ComponentPose=Local To Component(LocalPose=Use cached pose 'Base Pose'()), Alpha=1, GravityOverride:Vector="0,0,0", ExternalForce="0,0,0") -> (Pose) {
    angularDampingOverride=1
    angularSpringConstant=5000
    bAngularSpring=true
    bOverrideAngularDamping=true
    bUseGravityOverride=true
    gravityScale=0.35
    simulationSpace=BoneRelative
    alphaBoolBlend {blendOption=Linear}
    boundBone {boneName=robot_shoulderpad_r}
    componentAppliedLinearAccClamp=<3000 3000 3000>
    componentLinearAccScale=<0.15 0.15 0.15>
    componentLinearVelScale=<0.15 0.15 0.15>
    gravityOverride {z=-980}
    physicsBodyDefinitions [{
      boundBone {boneName=robot_shoulderpad_r}
      boxExtents=<25 15 15>
      constraintSetup {angularLimitsMax=<5 5 2>, angularLimitsMin=<-5 -5 -2>, angularTarget=<1 0 0>}
      localJointOffset=<2 0 0>
     }]
    relativeSpaceBone {boneName=upperarm_r}
    simSpaceSettings {maxAngularAcceleration=35, maxAngularVelocity=7, simSpaceAngularAlpha=0.2}
   }
   n2 = Anim Dynamics - Bone: robot_shoulderpad_l!(ComponentPose=@n1, Alpha=1, GravityOverride:Vector="0,0,0", ExternalForce="0,0,0") -> (Pose) {
    angularDampingOverride=1
    angularSpringConstant=5000
    bAngularSpring=true
    bOverrideAngularDamping=true
    bUseGravityOverride=true
    gravityScale=0.35
    simulationSpace=BoneRelative
    alphaBoolBlend {blendOption=Linear}
    boundBone {boneName=robot_shoulderpad_l}
    componentAppliedLinearAccClamp=<3000 3000 3000>
    componentLinearAccScale=<0.15 0.15 0.15>
    componentLinearVelScale=<0.15 0.15 0.15>
    gravityOverride {z=-980}
    physicsBodyDefinitions [{
      boundBone {boneName=robot_shoulderpad_l}
      boxExtents=<25 15 15>
      constraintSetup {angularLimitsMax=<5 5 2>, angularLimitsMin=<-5 -5 -2>, angularTarget=<1 0 0>}
      localJointOffset=<2 0 0>
     }]
    relativeSpaceBone {boneName=upperarm_l}
    simSpaceSettings {maxAngularAcceleration=35, maxAngularVelocity=7, simSpaceAngularAlpha=0.2}
   }
   n3 = Anim Dynamics - Bone: robot_shoulderpad_r(ComponentPose=Local To Component(LocalPose=Use cached pose 'Base Pose'()), Alpha=1, GravityOverride="0,0,0", ExternalForce="0,0,0") {
    angularDampingOverride=1
    angularSpringConstant=5000
    bAngularSpring=true
    bOverrideAngularDamping=true
    bUseGravityOverride=true
    gravityScale=0.35
    simulationSpace=BoneRelative
    alphaBoolBlend {blendOption=Linear}
    boundBone {boneName=robot_shoulderpad_r}
    componentAppliedLinearAccClamp=<2000 2000 2000>
    componentLinearAccScale=<0.15 0.15 0.15>
    componentLinearVelScale=<0.15 0.15 0.15>
    gravityOverride {z=-980}
    physicsBodyDefinitions [{
      boundBone {boneName=robot_shoulderpad_r}
      boxExtents=<25 15 15>
      constraintSetup {angularLimitsMax=<5 5 2>, angularLimitsMin=<-5 -5 -2>, angularTarget=<1 0 0>}
      localJointOffset=<2 0 0>
     }]
    relativeSpaceBone {boneName=upperarm_r}
   }
   n4 = Anim Dynamics - Bone: robot_shoulderpad_l(ComponentPose=@n3, Alpha=1, GravityOverride="0,0,0", ExternalForce="0,0,0") {
    angularDampingOverride=1
    angularSpringConstant=5000
    bAngularSpring=true
    bOverrideAngularDamping=true
    bUseGravityOverride=true
    gravityScale=0.35
    simulationSpace=BoneRelative
    alphaBoolBlend {blendOption=Linear}
    boundBone {boneName=robot_shoulderpad_l}
    componentAppliedLinearAccClamp=<2000 2000 2000>
    componentLinearAccScale=<0.15 0.15 0.15>
    componentLinearVelScale=<0.15 0.15 0.15>
    gravityOverride {z=-980}
    physicsBodyDefinitions [{
      boundBone {boneName=robot_shoulderpad_l}
      boxExtents=<25 15 15>
      constraintSetup {angularLimitsMax=<5 5 2>, angularLimitsMin=<-5 -5 -2>, angularTarget=<1 0 0>}
      localJointOffset=<2 0 0>
     }]
    relativeSpaceBone {boneName=upperarm_l}
   }
   Output Pose(Result=Blend Poses by bool(BlendPose_0=Component To Local(ComponentPose=@n2), BlendPose_1=Component To Local(ComponentPose=@n4), bActiveValue=Get bIsInFrontEnd())) {layerGroup=None}
   n5 = Layered blend per bone(
    BasePose=Copy Pose From Mesh() {bCopyCurves=true, bUseAttachedParent=true}
    BlendPoses_0=Copy Pose From Mesh(SourceMeshComponent=Get Body Mesh() -> (Body Mesh:Skeletal Mesh Component Object Reference))
   ) {layerSetup [{branchFilters [
       {boneName=clavicle_l}
       {boneName=robot_clavicle_l}
       {boneName=clavicle_r}
       {boneName=robot_clavicle_r}
       {boneName=upperarm_l}
       {boneName=robot_upperarm_l}
       {boneName=upperarm_r}
       {boneName=robot_upperarm_r}
       {boneName=upperarm_twist_01_l}
       {boneName=upperarm_twist_02_l}
       {boneName=upperarm_twist_01_r}
       {boneName=upperarm_twist_02_r}
       {boneName=robot_lowerarm_l}
       {boneName=lowerarm_l}
       {boneName=robot_lowerarm_r}
       {boneName=lowerarm_r}
       {boneName=lowerarm_twist_01_l}
       {boneName=lowerarm_twist_02_l}
       {boneName=lowerarm_twist_01_r}
       {boneName=lowerarm_twist_02_r}
       {boneName=hand_l}
       {boneName=hand_r}
       {boneName=index_metacarpal_l}
       {boneName=index_metacarpal_r}
       {boneName=index_01_l}
       {boneName=index_01_r}
       {boneName=index_02_l}
       {boneName=index_02_r}
       {boneName=index_03_l}
       {boneName=index_03_r}
       {boneName=middle_metacarpal_l}
       {boneName=middle_metacarpal_r}
       {boneName=middle_01_l}
       {boneName=middle_01_r}
       {boneName=middle_02_l}
       {boneName=middle_02_r}
       {boneName=middle_03_l}
       {boneName=middle_03_r}
       {boneName=pinky_metacarpal_l}
       {boneName=pinky_metacarpal_r}
       {boneName=pinky_01_l}
       {boneName=pinky_01_r}
       {boneName=pinky_02_l}
       {boneName=pinky_02_r}
       {boneName=pinky_03_l}
       {boneName=pinky_03_r}
       {boneName=ring_metacarpal_l}
       {boneName=ring_metacarpal_r}
       {boneName=ring_01_l}
       {boneName=ring_01_r}
       {boneName=ring_02_l}
       {boneName=ring_02_r}
       {boneName=ring_03_l}
       {boneName=ring_03_r}
       {boneName=thumb_01_l}
       {boneName=thumb_01_r}
       {boneName=thumb_02_l}
       {boneName=thumb_02_r}
       {boneName=thumb_03_l}
       {boneName=thumb_03_r}
       {boneName=robot_shoulderpad_l}
       {boneName=robot_shoulderpad_r}
      ]}]}
   Save cached pose 'Base Pose'(Pose=@n5) {cacheName=Base Pose}
  ]}
 EventGraph {nodes [
   n1 = Event OnCharacterCustomizationLoaded()
   > n2 = Set Body Mesh(Body Mesh:Skeletal Mesh Component Object Reference=Get Array Item with EFort Custom Part Type(Array=@n1.CustomMeshParts)) -> (Output_Get:Skeletal Mesh Component Object Reference)
   > AddTickPrerequisiteComponent!(execute, self:Actor Component Object Reference=GetOwningComponent(), PrerequisiteComponent:Actor Component Object Reference=@n2) -> (then)
   Event BlueprintUpdateAnimation()
  ]}
]
variables [Body Mesh {type=SkeletalMeshComponent}]
